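# A temporary ToyFox sound cue for the rift spawn effect, classed as player-critical boss fight audio.
# A mixer layers two rift start waves, close and distant, each through its own attenuation node.
# The close layer uses a close-range attenuation asset, while the distant layer uses a portal asset named as having no attenuation.
# The cue overrides concurrency to prevent new instances at its limit and has debug enabled.
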SoundCue /ToyFoxGameplay/Assets/Sound/TEMP/SC_TEMP_ToyFox_SFX_Spawn_Rift
root=Mixer_0
nodes [
 {id=Attenuation_1, children [WavePlayer_4], properties {attenuationSettings=&/SA_TEMP_ToyFox_SFX_Close}}
 {id=Attenuation_3, children [WavePlayer_5], properties {attenuationSettings=&/SA_TEMP_ToyFox_Portal_NoAtt}}
 {id=Mixer_0, children [Attenuation_1, Attenuation_3], properties {inputVolume [1,1]}}
 {id=WavePlayer_4, properties {soundWaveAssetPtr=&/Blondejaw_TEMP_ToyFox_SFX_Rift_Start_Close_01}}
 {id=WavePlayer_5, properties {soundWaveAssetPtr=&/Blondejaw_TEMP_ToyFox_SFX_Rift_Start_Distant_01}}
]
properties {
 bDebug=true
 bOverrideConcurrency=true
 maxDistance=150000
 soundClassObject=/Game/Sounds/FortSoundClassesAndModes/Fort_BossFight_SoundClass/Fort_BossFight_PlayerCritical
 volumeMultiplier=1.5
 attenuationOverrides {distanceAlgorithm=NaturalSound, falloffDistance=80000, lPFRadiusMin=17, stereoSpread=1000, attenuationShapeExtents {x=8000}}
 concurrencyOverrides {resolutionRule=PreventNew}
}
