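# Temporary ToyFox sound cue for a weakpoint being destroyed.
# A random node picks one of three equally weighted Blondejaw wave variations, which passes through a modulator at default settings and a short randomized delay before output.
# It references the ToyFox weakpoint attenuation asset and is routed to an instigator 2D hit-notification sound class.
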SoundCue /ToyFoxGameplay/Assets/Sound/TEMP/SC_ToyFox_SFX_Weakpoint_Destroyed
root=Delay_1
nodes [
 {id=Delay_1, children [Modulator_1], properties {delayMax=0.35, delayMin=0.25}}
 {id=Modulator_1, children [Random_0]}
 {id=Random_0, children [WavePlayer_0,WavePlayer_1,WavePlayer_2], properties {weights [1,1,1]}}
 {id=WavePlayer_0, properties {soundWaveAssetPtr=&/Blondejaw_ToyFox_SFX_Weakpoint_Destroyed_01}}
 {id=WavePlayer_1, properties {soundWaveAssetPtr=&/Blondejaw_ToyFox_SFX_Weakpoint_Destroyed_02}}
 {id=WavePlayer_2, properties {soundWaveAssetPtr=&/Blondejaw_ToyFox_SFX_Weakpoint_Destroyed_03}}
]
properties {
 attenuationSettings=&/SA_ToyFox_Weakpoint
 maxDistance=110000
 soundClassObject=&Fort_2DHitNotifications_Instigate_PSM
 volumeMultiplier=1
 attenuationOverrides {bAttenuate=false, falloffDistance=10000, stereoSpread=500, attenuationShapeExtents {x=2000}}
 concurrencyOverrides {maxCount=5, resolutionRule=PreventNew}
}
